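# Niagara smoke effect for the inverted BurnerCoil vehicle wheel cosmetic.
# One GPU sprite emitter steadily spawns a few large, translucent, randomly rotated puffs with upward-biased velocity.
# Over each short life, curves fade alpha in and out, grow size and scale velocity, while dynamic parameters feed the material.
# Shape Location runs in particle update, after the velocity solve, placing particles in a Y-offset cylinder volume.
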NiagaraSystem /VehicleCosmetics/Wheels/BurnerCoil_Inverted/PS/PS_Wheel_BurnerCoil_Inverted
dependencies {
 usedDataInterfaces [
  &NiagaraDataInterfaceCurve
  &NiagaraDataInterfaceVectorCurve
  &NiagaraDataInterfaceVector2DCurve
 ]
 usedDynamicInputs [&RandomRangeVector2D]
 usedModules [&SystemState]
}
emitters {Smoke {
  simTarget=GPUComputeSim
  emitterUpdateScript [
   &EmitterState()
   &SpawnRate(SpawnRate=6, Use Spawn Probability=false)
  ]
  particleSpawnScript [
   &InitializeParticle(
    Color=<rgba 1 1 1 0.2>
    Lifetime=&RandomRangeFloat(Minimum=0.8, Maximum=1.15)
    Position=&SimulationPosition()
    Position Mode=::NewEnumerator1
    Position Offset Coordinate Space=::Local
    Sprite Rotation Angle=&RandomRangeFloat(Minimum=0, Maximum=1)
    Sprite Rotation Mode=::NewEnumerator1
    Sprite Size Mode=::NewEnumerator3
    Uniform Sprite Size=&RandomRangeFloat(Minimum=110, Maximum=160)
    UsePositionOffset=true
    Write Material Random=true
   ) {module=InitializeParticle001}
   &CameraOffset(Camera Offset Amount=1)
   &AddVelocity(Velocity=&RandomRangeVector(Minimum=<0 -5 15>, Maximum=<0.01 5 50>))
  ]
  particleUpdateScript [
   &ScaleColor(
    Scale Alpha=&FloatFromCurve(FloatCurve=DataInterface() {propertyValues {Curve {keys [<at 0 0 RCIM_Cubic>,<at 0.1 1 RCIM_Cubic>,<at 1 0 RCIM_Cubic>]}}})
    ScaleA=true
    ScaleRGB=true
   )
   &DynamicMaterialParameters(
    Index 0 Param 1=&RandomRangeFloat(Minimum=0.01, Maximum=0.02)
    Index 0 Param 2=0.5
    Index 0 Param 3=2
    Module.Index0Float0Write=true
    Module.Index0Float1Write=true
    Module.Index0Float2Write=true
    Module.Index0Float3Write=false
   )
   &ScaleVelocity(Velocity Scale=&VectorFromCurve(VectorCurve=DataInterface() {propertyValues {XCurve {keys [<at 0 0 RCIM_Cubic>,<at 1 0>]}, YCurve {keys [<at 0 0.1 RCIM_Cubic>,<at 1 0.1>]}, ZCurve {keys [<at 0 0 RCIM_Cubic>,<at 1 0>]}}}))
   &ParticleState()
   &ScaleSpriteSize(Uniform Curve Sprite Scale=DataInterface() {propertyValues {Curve {keys [{tangentMode=RCTM_User, time=0, value=0.3}, {tangentMode=RCTM_User, time=1, value=1}]}}})
   &SpriteRotationRate(Rotation Rate=&RandomRangeFloat(Minimum=-0.02, Maximum=0.02))
   &SolveForcesAndVelocity(Clamp Velocity=false, Limit Acceleration=false)
   &ShapeLocation(
    Cylinder Height=10
    Cylinder Height Midpoint=0
    Cylinder Radius=35
    MaskSpawnGroup=false
    Offset=<0 30 0>
    Offset Mode=::NewEnumerator0
    Override Seed=false
    Rotation Axis=<0 1 0>
    Rotation Mode=::NewEnumerator1
    Shape Primitive=::NewEnumerator1
    Surface Only=false
    UseOwnerScale=true
   )
  ]
  rendererClasses [&NiagaraSpriteRendererProperties]
  renderers [{rendererClass=&NiagaraSpriteRendererProperties}]
 }}
summary {
 systemName=PS_Wheel_BurnerCoil_Inverted
 emitters [{emitterName=Smoke, simTarget=GPUComputeSim, rendererClasses [&NiagaraSpriteRendererProperties]}]
 userVariables []
}
